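; Manager that randomly chooses which Lesser Demon patrol paths and their paired Dragon Gate portals are active.
; On authority, once playlist data is ready, it draws from a cached copy of the path/portal list up to a hotfixable count, honoring per-entry hotfix flags and removing picks.
; ReshuffleDemonPortals disables every pair and rolls again.
; Numbered Activate/Deactivate cheat events toggle one of eight pairs by index through UpdateTargetPortal.
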(event Custom|ReshuffleDemonPortals
  (bind _activepathsamount_hf (Variables|Default|GetActivePathsAmount_Hf))
  (bind _returnvalue (ScalableFloat|GetValueatLevel _activepathsamount_hf))
  (bind _returnvalue_2 (Math|Float|Truncate _returnvalue))
  (bind _possiblepatrolpathscached (Variables|Default|GetPossiblePatrolPathsCached))
  (bind _possiblepatrolpathscached_2 (Variables|Default|GetPossiblePatrolPathsCached))
  (bind _randomindex (Variables|Default|GetRandomIndex))
  (for _array_element (Variables|Default|GetPossiblePatrolPaths)
    (bind (_array_element_patrolpath_10_2abb582a49eb549249f24b9d233f56da _array_element_portal_6_7d2657734256d9e3f6e2c18f28a20df1) (Utilities|Struct|BreakLesserDemonPathAndPortalStruct _array_element))
    (Utilities|IsValid _array_element_patrolpath_10_2abb582a49eb549249f24b9d233f56da
      (:Is%20Valid
        (Class|FortAthenaPatrolPath|SetPatrolPathEnabled _array_element_patrolpath_10_2abb582a49eb549249f24b9d233f56da false))
      (:Is%20Not%20Valid
        _))
    (Utilities|IsValid _array_element_portal_6_7d2657734256d9e3f6e2c18f28a20df1
      (:Is%20Valid
        (Variables|/Hermes/Props/ClayTile/ClayTile_DragonGate_B/Blueprints/BSG_ClayTile_DragonGate_B.BSG_ClayTile_DragonGate_B_C|SetPortalIsActive false _array_element_portal_6_7d2657734256d9e3f6e2c18f28a20df1))))
  (Variables|Default|SetPossiblePatrolPathsCached (Variables|Default|GetPossiblePatrolPaths))
  (if (> _returnvalue_2 0)
    (Utilities|FlowControl|ForLoopwithBreak 0 (- _returnvalue_2 1)
      (:LoopBody
        (if (> (Utilities|Array|Length _possiblepatrolpathscached) 0)
          (bind _output_get (Variables|Default|SetRandomIndex (Math|Random|RandomIntegerinRange 0 (Utilities|Array|LastIndex _possiblepatrolpathscached))))
          (if (Utilities|Array|IsValidIndex _possiblepatrolpathscached _output_get)
            (bind (_output_patrolpath_10_2abb582a49eb549249f24b9d233f56da _output_portal_6_7d2657734256d9e3f6e2c18f28a20df1 _output_enablehotfix_13_19ab4ad04b94ea03d9599390ddaa6d52) (Utilities|Struct|BreakLesserDemonPathAndPortalStruct (Utilities|Array|Get(acopy) _possiblepatrolpathscached _output_get)))
            (if (ScalableFloat|GetValueasBool _output_enablehotfix_13_19ab4ad04b94ea03d9599390ddaa6d52)
              (Utilities|IsValid _output_patrolpath_10_2abb582a49eb549249f24b9d233f56da
                (:Is%20Valid
                  _)
                (:Is%20Not%20Valid
                  (Utilities|Array|RemoveIndex _possiblepatrolpathscached_2 _randomindex)))
              (Class|FortAthenaPatrolPath|SetPatrolPathEnabled _output_patrolpath_10_2abb582a49eb549249f24b9d233f56da true)
              (Utilities|IsValid _output_portal_6_7d2657734256d9e3f6e2c18f28a20df1
                (:Is%20Valid
                  (Variables|/Hermes/Props/ClayTile/ClayTile_DragonGate_B/Blueprints/BSG_ClayTile_DragonGate_B.BSG_ClayTile_DragonGate_B_C|SetPortalIsActive true _output_portal_6_7d2657734256d9e3f6e2c18f28a20df1))
                (:Is%20Not%20Valid
                  _))
              (Utilities|Array|RemoveIndex _possiblepatrolpathscached_2 _randomindex))))
        (CallFunction|BreakPathActivationLoop)))))

(event EventBeginPlay
  (bind _activepathsamount_hf (Variables|Default|GetActivePathsAmount_Hf))
  (bind _returnvalue (ScalableFloat|GetValueatLevel _activepathsamount_hf))
  (bind _returnvalue_2 (Math|Float|Truncate _returnvalue))
  (bind _possiblepatrolpathscached (Variables|Default|GetPossiblePatrolPathsCached))
  (bind _possiblepatrolpathscached_2 (Variables|Default|GetPossiblePatrolPathsCached))
  (bind _randomindex (Variables|Default|GetRandomIndex))
  (Networking|SwitchHasAuthority
    (:Authority
      (Class|FortAsyncActionCurrentPlaylistReady|PlaylistDataReadyAsync
        (:OnReady
          (Variables|Default|SetPossiblePatrolPathsCached (Variables|Default|GetPossiblePatrolPaths))
          (if (> _returnvalue_2 0)
            (Utilities|FlowControl|ForLoopwithBreak 0 (- _returnvalue_2 1)
              (:LoopBody
                (if (> (Utilities|Array|Length _possiblepatrolpathscached) 0)
                  (bind _output_get (Variables|Default|SetRandomIndex (Math|Random|RandomIntegerinRange 0 (Utilities|Array|LastIndex _possiblepatrolpathscached))))
                  (if (Utilities|Array|IsValidIndex _possiblepatrolpathscached _output_get)
                    (bind (_output_patrolpath_10_2abb582a49eb549249f24b9d233f56da _output_portal_6_7d2657734256d9e3f6e2c18f28a20df1 _output_enablehotfix_13_19ab4ad04b94ea03d9599390ddaa6d52) (Utilities|Struct|BreakLesserDemonPathAndPortalStruct (Utilities|Array|Get(acopy) _possiblepatrolpathscached _output_get)))
                    (if (ScalableFloat|GetValueasBool _output_enablehotfix_13_19ab4ad04b94ea03d9599390ddaa6d52)
                      (Utilities|IsValid _output_patrolpath_10_2abb582a49eb549249f24b9d233f56da
                        (:Is%20Valid
                          _)
                        (:Is%20Not%20Valid
                          (Utilities|Array|RemoveIndex _possiblepatrolpathscached_2 _randomindex)))
                      (Class|FortAthenaPatrolPath|SetPatrolPathEnabled _output_patrolpath_10_2abb582a49eb549249f24b9d233f56da true)
                      (Utilities|IsValid _output_portal_6_7d2657734256d9e3f6e2c18f28a20df1
                        (:Is%20Valid
                          (Variables|/Hermes/Props/ClayTile/ClayTile_DragonGate_B/Blueprints/BSG_ClayTile_DragonGate_B.BSG_ClayTile_DragonGate_B_C|SetPortalIsActive true _output_portal_6_7d2657734256d9e3f6e2c18f28a20df1))
                        (:Is%20Not%20Valid
                          _))
                      (Utilities|Array|RemoveIndex _possiblepatrolpathscached_2 _randomindex))))
                (CallFunction|BreakPathActivationLoop)))))))))

(event Custom|BreakPathActivationLoop
  (bind _possiblepatrolpathscached (Variables|Default|GetPossiblePatrolPathsCached))
  (bind _possiblepatrolpathscached_2 (Variables|Default|GetPossiblePatrolPathsCached))
  (bind _randomindex (Variables|Default|GetRandomIndex))
  (Utilities|FlowControl|ForLoopwithBreak 0 (- (Math|Float|Truncate (ScalableFloat|GetValueatLevel (Variables|Default|GetActivePathsAmount_Hf))) 1)
    (:LoopBody
      (if (> (Utilities|Array|Length _possiblepatrolpathscached) 0)
        (bind _output_get (Variables|Default|SetRandomIndex (Math|Random|RandomIntegerinRange 0 (Utilities|Array|LastIndex _possiblepatrolpathscached))))
        (if (Utilities|Array|IsValidIndex _possiblepatrolpathscached _output_get)
          (bind (_output_patrolpath_10_2abb582a49eb549249f24b9d233f56da _output_portal_6_7d2657734256d9e3f6e2c18f28a20df1 _output_enablehotfix_13_19ab4ad04b94ea03d9599390ddaa6d52) (Utilities|Struct|BreakLesserDemonPathAndPortalStruct (Utilities|Array|Get(acopy) _possiblepatrolpathscached _output_get)))
          (if (ScalableFloat|GetValueasBool _output_enablehotfix_13_19ab4ad04b94ea03d9599390ddaa6d52)
            (Utilities|IsValid _output_patrolpath_10_2abb582a49eb549249f24b9d233f56da
              (:Is%20Valid
                _)
              (:Is%20Not%20Valid
                (Utilities|Array|RemoveIndex _possiblepatrolpathscached_2 _randomindex)))
            (Class|FortAthenaPatrolPath|SetPatrolPathEnabled _output_patrolpath_10_2abb582a49eb549249f24b9d233f56da true)
            (Utilities|IsValid _output_portal_6_7d2657734256d9e3f6e2c18f28a20df1
              (:Is%20Valid
                (Variables|/Hermes/Props/ClayTile/ClayTile_DragonGate_B/Blueprints/BSG_ClayTile_DragonGate_B.BSG_ClayTile_DragonGate_B_C|SetPortalIsActive true _output_portal_6_7d2657734256d9e3f6e2c18f28a20df1))
              (:Is%20Not%20Valid
                _))
            (Utilities|Array|RemoveIndex _possiblepatrolpathscached_2 _randomindex))))
      (CallFunction|BreakPathActivationLoop))))

(event Custom|ActivatePortal01
  (bind _output_get (Variables|Default|SetCheatTargetedPortalIndex 0))
  (CallFunction|UpdateTargetPortal true _output_get))

(event Custom|DeactivatePortal01
  (bind _output_get (Variables|Default|SetCheatTargetedPortalIndex 0))
  (CallFunction|UpdateTargetPortal false _output_get))

(event Custom|ActivatePortal02
  (bind _output_get (Variables|Default|SetCheatTargetedPortalIndex 1))
  (CallFunction|UpdateTargetPortal true _output_get))

(event Custom|DeactivatPortal02
  (bind _output_get (Variables|Default|SetCheatTargetedPortalIndex 1))
  (CallFunction|UpdateTargetPortal false _output_get))

(event Custom|ActivatePortal03
  (bind _output_get (Variables|Default|SetCheatTargetedPortalIndex 2))
  (CallFunction|UpdateTargetPortal true _output_get))

(event Custom|DeactivatePortal03
  (bind _output_get (Variables|Default|SetCheatTargetedPortalIndex 2))
  (CallFunction|UpdateTargetPortal false _output_get))

(event Custom|ActivatePortal04
  (bind _output_get (Variables|Default|SetCheatTargetedPortalIndex 3))
  (CallFunction|UpdateTargetPortal true _output_get))

(event Custom|DeactivatePortal04
  (bind _output_get (Variables|Default|SetCheatTargetedPortalIndex 3))
  (CallFunction|UpdateTargetPortal false _output_get))

(event Custom|ActivatePortal05
  (bind _output_get (Variables|Default|SetCheatTargetedPortalIndex 4))
  (CallFunction|UpdateTargetPortal true _output_get))

(event Custom|DeactivatePortal05
  (bind _output_get (Variables|Default|SetCheatTargetedPortalIndex 4))
  (CallFunction|UpdateTargetPortal false _output_get))

(event Custom|ActivatePortal06
  (bind _output_get (Variables|Default|SetCheatTargetedPortalIndex 5))
  (CallFunction|UpdateTargetPortal true _output_get))

(event Custom|DeactivatePortal06
  (bind _output_get (Variables|Default|SetCheatTargetedPortalIndex 5))
  (CallFunction|UpdateTargetPortal false _output_get))

(event Custom|ActivatePortal07
  (bind _output_get (Variables|Default|SetCheatTargetedPortalIndex 6))
  (CallFunction|UpdateTargetPortal true _output_get))

(event Custom|DeactivatePortal07
  (bind _output_get (Variables|Default|SetCheatTargetedPortalIndex 6))
  (CallFunction|UpdateTargetPortal false _output_get))

(event Custom|ActivatePortal08
  (bind _output_get (Variables|Default|SetCheatTargetedPortalIndex 7))
  (CallFunction|UpdateTargetPortal true _output_get))

(event Custom|DeactivatePortal08
  (bind _output_get (Variables|Default|SetCheatTargetedPortalIndex 7))
  (CallFunction|UpdateTargetPortal false _output_get))

(fn UpdateTargetPortal
  (params (Activate :type Boolean) (Index :type Integer))
  (bind _possiblepatrolpaths (Variables|Default|GetPossiblePatrolPaths))
  (bind (_output_patrolpath_10_2abb582a49eb549249f24b9d233f56da _output_portal_6_7d2657734256d9e3f6e2c18f28a20df1) (Utilities|Struct|BreakLesserDemonPathAndPortalStruct (Utilities|Array|Get(acopy) _possiblepatrolpaths Index)))
  (Utilities|IsValid _output_patrolpath_10_2abb582a49eb549249f24b9d233f56da
    (:Is%20Valid
      (Class|FortAthenaPatrolPath|SetPatrolPathEnabled _output_patrolpath_10_2abb582a49eb549249f24b9d233f56da (Variables|GetActivate)))
    (:Is%20Not%20Valid
      _))
  (Utilities|IsValid _output_portal_6_7d2657734256d9e3f6e2c18f28a20df1
    (:Is%20Valid
      (Variables|/Hermes/Props/ClayTile/ClayTile_DragonGate_B/Blueprints/BSG_ClayTile_DragonGate_B.BSG_ClayTile_DragonGate_B_C|SetPortalIsActive (Variables|GetActivate) _output_portal_6_7d2657734256d9e3f6e2c18f28a20df1))))

(fn ConstructionScript)
